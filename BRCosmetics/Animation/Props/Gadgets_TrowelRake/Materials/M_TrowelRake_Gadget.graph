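# Cloth-shaded anime-style material for the TrowelRake gadget cosmetic prop, built on the 2023 Anime Parent shader's Pre, Calculations and Post functions.
# A 2x2 flipbook of the TrowelRake text texture, stepped by MaterialCurveA with adjustable UV offset and scale, tints lit and shaded colors through its alpha.
# Those colors pass through the shared eye-animation function before final shading.
# An IsKicks static switch optionally applies the sock function to the result.
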Material /BRCosmetics/Animation/Props/Gadgets_TrowelRake/Materials/M_TrowelRake_Gadget
graph {
 properties {
  bAutomaticallySetUsageInEditor=false
  bUseMaterialAttributes=true
  bUsedWithClothing=true
  bUsedWithMeshDeformer=true
  bUsedWithMorphTargets=true
  bUsedWithNiagaraMeshParticles=true
  bUsedWithSkeletalMesh=true
  bUsedWithStaticLighting=true
  pixelDepthOffsetMode=PDOM_Legacy
  shadingModel=MSM_Cloth
  usedShadingModels=Cloth
 }
 nodes [
  n1 = MaterialFunctionCall!(CustomUV) {materialFunction=/Game/Global/Materials/Character_Parent_2023/Anime_Parent_2023/MaterialFunctions/MFC_AnimeShader_2023_Pre}
  n2 = MaterialFunctionCall!(CustomLightVector, OverrideNormal_WS, STT=@n1.STT, Normal_Map=@n1.STT, DFL=@n1.STT, FX=@n1.STT) {materialFunction=/Game/Global/Materials/Character_Parent_2023/Anime_Parent_2023/MaterialFunctions/MFC_AnimeShader_2023_Calculations}
  n3 = ScalarParameter() {group=Flipbook, parameterName=MaterialCurveA, sliderMax=1}
  n4 = Constant() {r=2}
  n5 = Constant() {r=2}
  n6 = Multiply(A=@n4, B=@n5)
  n7 = VectorParameter() {group=Flipbook, parameterName=Flipbook UV Position}
  n8 = MaterialFunctionCall(
   UVs=Add(A=TextureCoordinate(), B=ComponentMask(None=@n7.RGB) {g=true, r=true})
   Texture Scale=ScalarParameter() {defaultValue=1, group=Flipbook, parameterName=Flipbook UV Scale}
  ) {materialFunction=&ScaleUVsByCenter}
  n9 = Multiply(
   A=MaterialFunctionCall(
    "Animation  Phase (0-1)"=Divide(A=@n3, B=@n6)
    Number of Rows=@n4
    Number of Columns=@n5
    Texture=TextureObject() {texture=/BRCosmetics/Animation/Props/Gadgets_TrowelRake/Textures/T_TrowelRake_Text_D}
    UVs=@n8.UVs
   ) {materialFunction=&FlipBook}.Alpha
   B=@n8.0-1 mask
  )
  n10 = VertexColor()
  n11 = TextureSampleParameter2D() {
   group=01_EyeAnimation_Parameters
   parameterName=EyeMask
   samplerSource=SSM_Wrap_WorldGroupSettings
   samplerType=SAMPLERTYPE_Masks
   sortPriority=-32
   texture=&T_WhiteMask
  }
  n12 = StaticSwitchParameter(True=AppendVector(A=@n10, B=@n10), False=@n11.RGBA) {group=01_EyeAnimation_Parameters, parameterName=UseVertexColorForEyeMask, sortPriority=-33}
  n13 = TextureCoordinate() {coordinateIndex=1}
  n14 = Sign(None=Subtract(A=ComponentMask(None=@n13) {r=true}) {constB=0.0001})
  n15 = MaterialFunctionCall!(
   Color Lit=LinearInterpolate(A=@n1.Color_Lit, B=VectorParameter() {parameterName=Lit Color - Flipbook, defaultValue=<rgba 0.0091 0.0097 0.0091 1>}.RGB, Alpha=@n9)
   Color Shaded=LinearInterpolate(A=@n1.Color_Shaded, B=VectorParameter() {parameterName=Shaded Color - Flipbook, defaultValue=<rgba 0 0 0 1>}.RGB, Alpha=@n9)
   Use UV2=StaticBoolParameter() {defaultValue=true, group=01_EyeAnimation_Parameters, parameterName=Use UV2 for Eye Animation, sortPriority=-30}
   Mask=StaticSwitchParameter(
    True=ChannelMaskParameter(None=@n12) {group=01_EyeAnimation_Parameters, parameterName=EyeMaskChannel, sortPriority=-32}.R
    False=Saturate(None=@n14)
   ) {group=01_EyeAnimation_Parameters, parameterName=UseEyeMask, sortPriority=-33}
  ) {materialFunction=/Game/Global/Materials/Character_Parent_2023/Anime_Parent_2023/MaterialFunctions/MF_AnimeEyeAnimation_v2}
  n16 = MaterialFunctionCall!(
   Shading_Gradient=@n2.Shading_Gradient
   Threshold_Gradient=@n2.Shading_Gradient
   Specular_Gradient=@n2.Shading_Gradient
   Sheen_Mask=@n2.Shading_Gradient
   Thickness=@n2.Shading_Gradient
   DFL_Mask=@n2.Shading_Gradient
   FX=@n2.Shading_Gradient
   Metallic_Gradient=@n2.Shading_Gradient
   RimLight_Gradient=@n2.Shading_Gradient
   NormalBuffer=@n2.Shading_Gradient
   --------
   CINE_LightSources=@n2.Shading_Gradient
   CINE_LightColor=@n2.Shading_Gradient
   ---------
   Color_Lit=@n15.Lit
   Color_Shaded=@n15.Lit
   SRM=@n1.SRM
   Emissive=@n1.SRM
   DFL_Color=@n1.SRM
   Specular_Color=@n1.SRM
   RimLight_Color=@n1.SRM
  ) {materialFunction=/Game/Global/Materials/Character_Parent_2023/Anime_Parent_2023/MaterialFunctions/MFC_AnimeShader_2023_Post}
  StaticSwitchParameter(
   True=MaterialFunctionCall!(InOpacity, Material=@n16.Result, Mask) {materialFunction=/Game/Global/Materials/Cosmetics/Kicks/Socks/Materials/MF_Sock}.Result
   False=@n16.Result
  ) {group=Kicks, parameterName=IsKicks}
  TemporalResponsivenessOutput!(None)
 ]
}
