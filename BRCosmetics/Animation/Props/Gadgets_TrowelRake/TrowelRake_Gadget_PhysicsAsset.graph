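# Physics asset for the TrowelRake gadget prop, providing its skeletal mesh with a simple collision and physics shape.
# A single capsule body on the JNT_Sign bone, offset and rotated from the joint, is the only shape.
# It has no constraints, so there is no jointed ragdoll chain.
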PhysicsAsset /BRCosmetics/Animation/Props/Gadgets_TrowelRake/TrowelRake_Gadget_PhysicsAsset
bodies {JNT_Sign [{shapeType=Capsule, length=46.1448, radius=38.389, center=<-4.4157 -0.8715 48.05>, rotation=<rot -1.1183 -90 15.1309>}]}
constraints []
